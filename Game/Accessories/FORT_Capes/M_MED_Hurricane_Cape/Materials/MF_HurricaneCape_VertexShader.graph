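# Material function for the Hurricane cape that animates cloth with vertex displacement and moving fold normals on incoming material attributes.
# Pre-skinned position divided by a base scale becomes planar coordinates for panning textures.
# Vertex-color-red-weighted Perlin noise, plus normalized position, drives World Position Offset.
# A noise-distorted folds normal, flattened by vertex red, blends into the input normal; the layered sine-wave displacement is left unconnected.
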MaterialFunction /Game/Accessories/FORT_Capes/M_MED_Hurricane_Cape/Materials/MF_HurricaneCape_VertexShader
graph {nodes [
  n1 = PreSkinnedPosition()
  n2 = Normalize(VectorInput=@n1)
  n3 = MaterialFunctionCall(Float3=Divide(A=@n1, B=VectorParameter() {parameterName=HC_BaseScale, defaultValue=<rgba 70 12 115 1>}.RGB)) {materialFunction=&BreakOutFloat3Components}
  n4 = ComponentMask(None=MaterialFunctionCall(X=@n3.R, Y=@n3.R, Z=OneMinus(None=@n3.B)) {materialFunction=&MakeFloat3}.Result) {b=true, r=true}
  n5 = ComponentMask(None=@n4) {g=true}
  n6 = Time()
  n7 = Multiply(A=@n6, B=ScalarParameter() {defaultValue=-0.25, parameterName=WPO_SineWaveSpeed_01})
  n8 = Sine(None=Multiply(A=Add(A=@n5, B=@n7), B=ScalarParameter() {defaultValue=1, parameterName=HC_SineWave_Frequency_01}))
  n9 = Add(A=@n6, B=ScalarParameter() {defaultValue=0.25, parameterName=WPO_SineWaveOffset_02})
  n10 = ScalarParameter() {defaultValue=-0.5, parameterName=WPO_SineWaveSpeed_02}
  n11 = Multiply(A=Add(A=@n5, B=Multiply(A=@n9, B=@n10)), B=ScalarParameter() {defaultValue=2, parameterName=HC_SineWave_Frequency_02})
  n12 = Add(
   A=Multiply(A=@n8, B=ScalarParameter() {defaultValue=5, parameterName=SineWave_Displacement_Amt_01})
   B=Multiply(A=Sine(None=@n11), B=ScalarParameter() {defaultValue=2, parameterName=SineWave_Displacement_Amt_02})
  )
  n13 = VertexColor()
  Add(
   A=@n2
   B=Multiply(A=Multiply(A=@n13.R, B=@n12), B=VectorParameter() {parameterName=SineWaveDispalcementDirection, defaultValue=<rgba 0.25 0.25 0 1>}.RGB)
  )
  ConstantBiasScale(None=@n8)
  n14 = GetMaterialAttributes(None=FunctionInput() {inputType=FunctionInput_MaterialAttributes}) {attributeGetTypes [#1]}
  n15 = VertexInterpolator(VS=@n4)
  n16 = ComponentMask(None=VectorParameter() {parameterName=HC_FoldsOffset, defaultValue=<rgba 0.25 0 0 1>}.RGB) {g=true, r=true}
  n17 = VectorParameter() {parameterName=HC_FoldsTiling, defaultValue=<rgba 1 1.25 0 1>}
  n18 = VectorParameter() {parameterName=HC_FoldsDistortionNoise_Tiling, defaultValue=<rgba 1 1 0 1>}
  n19 = Panner(
   Coordinate=Multiply(A=@n15.PS, B=ComponentMask(None=@n18.RGB) {g=true, r=true})
   Speed=ComponentMask(None=VectorParameter() {parameterName=HC_FoldsDistortionNoise_PanSpeedAndDirection, defaultValue=<rgba 0 -0.25 0 1>}.RGB) {g=true, r=true}
  )
  n20 = Subtract(A=Multiply(A=TextureSample(UVs=@n19) {texture=/Game/Global/Textures/Default/Noise/3xPerlin_M}.RGB) {constB=2})
  n21 = ScalarParameter() {defaultValue=0.05, parameterName=HC_GoldsDistortion_Strength}
  n22 = Add(
   A=Multiply(A=Add(A=@n15.PS, B=@n16), B=ComponentMask(None=@n17.RGB) {g=true, r=true})
   B=ComponentMask(None=Multiply(A=@n20, B=@n21)) {g=true, r=true}
  )
  n23 = VectorParameter() {parameterName=HC_FoldsPanSpeedAndDirection, defaultValue=<rgba 0 -0.25 0 1>}
  n24 = MaterialFunctionCall(
   Normal=TextureSample(UVs=Panner(Coordinate=@n22, Speed=@n23.RGB)) {samplerType=SAMPLERTYPE_Normal, texture=/Game/Gadgets/Assets/VinderTech_GliderChute/Glider_BlackMondayCape/Textures/T_Folds_N}.RGB
   Flatness=LinearInterpolate(B=ScalarParameter() {defaultValue=0.75, parameterName=HC_FoldsFlattenMin}, Alpha=Power(Base=@n13.R)) {constA=1, constB=0.25}
  ) {materialFunction=&FlattenNormal}
  n25 = ComponentMask(None=VectorParameter() {parameterName=HC_WPONoise_Tiling, defaultValue=<rgba 1 1 0 1>}.RGB) {g=true, r=true}
  n26 = VectorParameter() {parameterName=HC_WPONoise_PanSpeedAndDirection, defaultValue=<rgba 0 -0.15 0 1>}
  n27 = TextureSample(UVs=Panner(Coordinate=Multiply(A=@n4, B=@n25), Speed=ComponentMask(None=@n26.RGB) {g=true, r=true})) {texture=&3xPerlin_M}
  n28 = Subtract(A=Multiply(A=@n27.RGB) {constB=2})
  n29 = VectorParameter() {parameterName=HC_WPO_Offset_Direction, defaultValue=<rgba 4 4 0 1>}
  n30 = Add(A=@n2, B=Multiply(A=@n13.R, B=Multiply(A=@n28, B=@n29.RGB)))
  FunctionOutput(None=SetMaterialAttributes!(
    MaterialAttributes=@n14.MaterialAttributes
    Normal=MaterialFunctionCall(BaseNormal=@n14.Normal, AdditionalNormal=@n24.Result) {materialFunction=&BlendAngleCorrectedNormals}.Result
    WorldPositionOffset=Add(A=@n30) {constB=0}
   ) {attributeSetTypes [#1, #2]})
  MaterialFunctionCall() {materialFunction=&ObjectLocalBounds}
  ObjectRadius()
  VertexInterpolator(VS=@n27.RGB)
 ]}
