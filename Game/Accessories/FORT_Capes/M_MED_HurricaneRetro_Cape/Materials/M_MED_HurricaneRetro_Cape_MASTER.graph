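# Master character material for the HurricaneRetro cape accessory, built on material attributes for skeletal meshes, clothing and mesh deformers.
# Shared character PreFX feeds CharacterSkinFX, whose result goes into PostFX for the final output.
# A UseClothVertexShader static switch optionally inserts the vertex shader function borrowed from the original Hurricane cape before PostFX.
# The shading model comes from the material graph rather than a fixed setting.
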Material /Game/Accessories/FORT_Capes/M_MED_HurricaneRetro_Cape/Materials/M_MED_HurricaneRetro_Cape_MASTER
graph {
 properties {
  bUseMaterialAttributes=true
  bUsedWithClothing=true
  bUsedWithMeshDeformer=true
  bUsedWithSkeletalMesh=true
  pixelDepthOffsetMode=PDOM_Legacy
  refractionMethod=RM_IndexOfRefraction
  shadingModel=MSM_FromMaterialExpression
  usedShadingModels=Default Lit | Subsurface | Clear Coat | Subsurface Profile | Cloth | Eye
 }
 nodes [
  n1 = MaterialFunctionCall() {materialFunction=&BaseFN_CharacterFunction_PreFX}
  n2 = MaterialFunctionCall(
   PreskinnedPosition=@n1.Pre skinned Local Position
   Gmap=@n1.Pre skinned Local Position
   MaterialAttributes=@n1.Pre skinned Local Position
   AO=@n1.Pre skinned Local Position
  ) {materialFunction=&CharacterSkinFX}
  MaterialFunctionCall(
   AO=@n1.AO
   Distance Blend=@n1.AO
   Fresnel=@n1.AO
   Material=StaticSwitchParameter(
    True=MaterialFunctionCall!(In=@n2.Result) {materialFunction=/Game/Accessories/FORT_Capes/M_MED_Hurricane_Cape/Materials/MF_HurricaneCape_VertexShader}.Result
    False=@n2.Result
   ) {parameterName=UseClothVertexShader}
  ) {materialFunction=&BaseFN_CharacterFunction_PostFX}
 ]
}
